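# Physics asset for the EmeraldGlassStandAlone backpack cosmetic, giving capsule bodies to a small pair of legs on the pack.
# The dyn_pelvis body, built from two capsules, is angularly locked to the pack body and roots left and right thigh, calf and foot chains.
# Knees lock twist and one swing axis to act as hinges, while hips and ankles have small angle values that differ slightly between sides.
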PhysicsAsset /BRCosmetics/Accessories/FORT_Backpacks/Backpack_M_MED_EmeraldGlassStandAlone/Meshes/M_MED_EmeraldGlassStandAlone_Pack_Leg_Physics
bodies {
 dyn_calf_l [{shapeType=Capsule, length=4.1281, radius=2.3733, shapeName=dyn_calf_l_capsule, center=<2.7869 0 0.3255>, rotation=<rot 0 90 90>}]
 dyn_calf_r [{shapeType=Capsule, length=4.1281, radius=2.1233, shapeName=dyn_calf_r_capsule, center=<-2.7869 0 0.3255>, rotation=<rot 0 90 90>}]
 dyn_foot_l [{
   shapeType=Capsule
   length=3.75
   radius=1.8347
   shapeName=dyn_foot_l_capsule
   center=<2.1185 -0.0631 -1.7596>
   rotation=<rot -0.0002 90.0009 -0.1865>
  }]
 dyn_foot_r [{
   shapeType=Capsule
   length=3.75
   radius=1.8347
   shapeName=dyn_foot_l_capsule
   center=<-2.7026 0.0631 2.0934>
   rotation=<rot -0.0002 90.001 -177.4102>
  }]
 dyn_pelvis [
  {shapeType=Capsule, length=7, radius=3.25, center=<0 0.6216 -3.3481>, rotation=<rot 90 0 0>}
  {shapeType=Capsule, length=11.5, radius=6.75, shapeName=dyn_pelvis_capsule, center=<0 2.8909 -11.9225>, rotation=<rot 90 0 0>}
 ]
 dyn_thigh_l [{shapeType=Capsule, length=3.9989, radius=2.6448, shapeName=dyn_thigh_l_capsule, center=<5.3659 0 -0.081>, rotation=<rot 0 90 90>}]
 dyn_thigh_r [{shapeType=Capsule, length=3.9989, radius=2.6448, shapeName=dyn_thigh_r_capsule, center=<-5.3659 0 -0.081>, rotation=<rot 0 90 -90>}]
 pack [{shapeType=Capsule, length=1, radius=1}]
}
constraints [
 {
  bone1Name=dyn_pelvis
  bone2Name=pack
  swing1LimitDegrees=45
  swing1Motion=Locked
  swing2LimitDegrees=45
  swing2Motion=Locked
  twistLimitDegrees=45
  twistMotion=Locked
 }
 {bone1Name=dyn_thigh_l, bone2Name=dyn_pelvis, swing1LimitDegrees=5, swing2LimitDegrees=5, twistLimitDegrees=10}
 {
  bone1Name=dyn_calf_l
  bone2Name=dyn_thigh_l
  swing1LimitDegrees=20
  swing1Motion=Locked
  swing2LimitDegrees=5
  twistLimitDegrees=20
  twistMotion=Locked
 }
 {bone1Name=dyn_foot_l, bone2Name=dyn_calf_l, swing1LimitDegrees=10, swing2LimitDegrees=10, twistLimitDegrees=10}
 {bone1Name=dyn_thigh_r, bone2Name=dyn_pelvis, swing1LimitDegrees=10, swing2LimitDegrees=5, twistLimitDegrees=10}
 {
  bone1Name=dyn_calf_r
  bone2Name=dyn_thigh_r
  swing1LimitDegrees=20
  swing1Motion=Locked
  swing2LimitDegrees=5
  twistLimitDegrees=20
  twistMotion=Locked
 }
 {bone1Name=dyn_foot_r, bone2Name=dyn_calf_r, swing1LimitDegrees=10, swing2LimitDegrees=25, twistLimitDegrees=10}
]
