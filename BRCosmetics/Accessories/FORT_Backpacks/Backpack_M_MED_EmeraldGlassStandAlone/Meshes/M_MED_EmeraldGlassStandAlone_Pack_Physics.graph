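# Physics asset for the EmeraldGlassStandAlone backpack cosmetic, defining simple bodies for its dynamic head, spine and pelvis bones.
# Four capsule bodies form a chain from the small pack body through pelvis and spine to head.
# The pelvis-to-pack joint locks swing and twist, while the spine and head joints carry limits of a few degrees, slightly looser at the head.
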PhysicsAsset /BRCosmetics/Accessories/FORT_Backpacks/Backpack_M_MED_EmeraldGlassStandAlone/Meshes/M_MED_EmeraldGlassStandAlone_Pack_Physics
bodies {
 dyn_head [{shapeType=Capsule, length=5.2401, radius=4.4488, shapeName=dyn_head_capsule, center=<0 0 5>, rotation=<rot 0 0 90>}]
 dyn_pelvis [{shapeType=Capsule, length=7, radius=3.25, center=<0 0.6216 -3.3481>, rotation=<rot 90 0 0>}]
 dyn_spine_01 [{shapeType=Capsule, length=5.6785, radius=4.7848, shapeName=dyn_spine_01_capsule, center=<-0.0514 0 -2.2561>, rotation=<rot 0 90 -90>}]
 pack [{shapeType=Capsule, length=1, radius=1}]
}
constraints [
 {
  bone1Name=dyn_pelvis
  bone2Name=pack
  swing1LimitDegrees=45
  swing1Motion=Locked
  swing2LimitDegrees=45
  swing2Motion=Locked
  twistLimitDegrees=45
  twistMotion=Locked
 }
 {bone1Name=dyn_spine_01, bone2Name=dyn_pelvis, swing1LimitDegrees=1, swing2LimitDegrees=1, twistLimitDegrees=3}
 {bone1Name=dyn_head, bone2Name=dyn_spine_01, swing1LimitDegrees=4, swing2LimitDegrees=4, twistLimitDegrees=4}
]
