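# Animation Blueprint for the EmeraldGlassStandAlone backpack cosmetic, layering animated idles with physics-driven secondary motion.
# A random player picks idle or idle-break clips, applied additively over a base pose, then two chained RigidBody nodes simulate the pack and leg physics assets in root-bone space.
# The update event sets upward external force while falling, lowers Alpha when skydiving or parachuting, and derives UpperBodyAlpha from a blend-off curve.
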AnimBlueprint /BRCosmetics/Accessories/FORT_Backpacks/Backpack_M_MED_EmeraldGlassStandAlone/Meshes/M_MED_EmeraldGlassStandAlone_Pack_AnimBP
target_skeleton=&/M_MED_EmeraldGlassStandAlone_Pack_Skeleton
graphs [
 AnimGraph {nodes [
   Input Pose()
   n1 = Random Sequence Player!() -> (Pose:Pose Link Structure) {
    is=title AnimGraphNode_RandomPlayer
    entries [
     {
      chanceToPlay=0.7
      maxLoopCount=0
      maxPlayRate=1
      minLoopCount=0
      minPlayRate=1
      sequence=/BRCosmetics/Accessories/FORT_Backpacks/Backpack_M_MED_EmeraldGlassStandAlone/Animation/EmeraldGlassStandAlone_Idle_PROP
      blendIn {blendOption=Linear, blendTime=0.2, customCurve=null}
     }
     {
      chanceToPlay=0.3
      maxLoopCount=0
      maxPlayRate=1
      minLoopCount=0
      minPlayRate=1
      sequence=/BRCosmetics/Accessories/FORT_Backpacks/Backpack_M_MED_EmeraldGlassStandAlone/Animation/EmeraldGlassStandAlone_IdleBreak_PROP
      blendIn {blendOption=Linear, blendTime=0.2, customCurve=null}
     }
    ]
   }
   n2 = Sequence Player 'EmeraldGlassStandAlone_BasePose_PROP'() {
    bLoopAnimation=false
    sequence=/BRCosmetics/Accessories/FORT_Backpacks/Backpack_M_MED_EmeraldGlassStandAlone/Animation/EmeraldGlassStandAlone_BasePose_PROP
   }
   n3 = RigidBody!(
    ComponentPose=Local To Component(LocalPose=Apply Additive(Base=@n2, Additive=@n1) {alphaBoolBlend {blendOption=Linear}})
    Alpha
    OverrideWorldGravity:Vector="0,0,-980"
    ExternalForce
   ) -> (Pose) {
    bClampLinearTranslationLimitToRefPose=true
    bOverrideWorldGravity=true
    overridePhysicsAsset=&/M_MED_EmeraldGlassStandAlone_Pack_Physics
    simulationSpace=BaseBoneSpace
    alphaBoolBlend {blendOption=Linear}
    baseBoneRef {boneName=root}
    componentAppliedLinearAccClamp=<2000 2000 2000>
    componentLinearAccScale=<0.36 0.36 0.36>
    componentLinearVelScale=<0.36 0.36 0.36>
   }
   n4 = RigidBody(ComponentPose=@n3, OverrideWorldGravity="0,0,-980") {
    bClampLinearTranslationLimitToRefPose=true
    bOverrideWorldGravity=true
    overridePhysicsAsset=&/M_MED_EmeraldGlassStandAlone_Pack_Leg_Physics
    simulationSpace=BaseBoneSpace
    alphaBoolBlend {blendOption=Linear}
    baseBoneRef {boneName=root}
    componentAppliedLinearAccClamp=<2000 2000 2000>
    componentLinearAccScale=<0.36 0.36 0.36>
    componentLinearVelScale=<0.36 0.36 0.36>
   }
   Output Pose(Result=Component To Local(ComponentPose=@n4)) {layerGroup=None}
  ]}
 EventGraph {nodes [
   Event BlueprintUpdateAnimation()
   > n1 = Sequence()
   > n2 = Branch(Condition=Get bIsFalling())
   > Set External Force(External Force:Vector="0, 0,3000") -> (Output_Get:Vector)
   Set External Force(execute=@n2.else, External Force="0, 0, 0")
   n3 = Branch(execute=@n1.then_1, Condition=OR Boolean(A=Get bIsSkyDiving(), B=Get bIsParachuteOpened()))
   > n4 = Set Alpha(Alpha:float=0.3) -> (Output_Get:float)
   n5 = float * float(A=float - float(A=1, B=GetCurveValue(CurveName=UpperBodyPhysicsBlendOff)), B=1) {is=title K2Node_PromotableOperator}
   > Set UpperBodyAlpha(UpperBodyAlpha:float=float * float(A=@n4, B=@n5)) -> (Output_Get:float)
   n6 = Set Alpha(execute=@n3.else, Alpha=1)
   > Set UpperBodyAlpha(UpperBodyAlpha=float * float(A=@n6, B=@n5))
   TryGetPawnOwner()
  ]}
]
variables [
 Alpha {type=number, default=1}
 External Force {type=Vector}
 Gravity Override Params {type=GravityOverrideParamsStruct, default {ziplining {windAmplitude=<100 -500 300>, windFrequency=<1.5 7 1.5>}}}
 Override World Gravity {type=Vector, default=<0 0 -980>}
 UpperBodyAlpha {type=number, default=1}
]
